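; BP_BunkerDoor_TidalCrane_Static is a bunker door Blueprint in the CrewCut set, apparently a static variant for the TidalCrane location, that removes itself at runtime.
; On BeginPlay it waits through a very short delay and then destroys its own actor.
; The construction script is empty, and the graph holds no other logic.
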(event EventBeginPlay
  (Utilities|FlowControl|Delay 0.1)
  (Actor|DestroyActor))

(fn ConstructionScript)
