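; Static-mesh entrance Blueprint for the CrewCut set that colors its spotlight and emissive frame red by default, with green for CherrySmoke hero rank A.
; On BeginPlay it waits for playlist data, then cosmetically calls UpdateLight.
; A first-load flag gates the local player's rank-tag check, intended to handle relevancy changes.
; UpdateLight, gated by that flag, sets the spotlight color and dynamic material parameter, then prints a debug string.
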(event EventBeginPlay
  ; Default all frames to red
  (Class|FortAsyncActionCurrentPlaylistReady|PlaylistDataReadyAsync
    (:OnReady
      (Networking|CanExecuteCosmeticEvents
        (:True
          (if (Variables|Default|GetFirstTimeLoadingThisMatch?)
            ;; Checking for Rank A for in and out of relevancy changes
            (if (GameplayTags|HasMatchingGameplayTag (Game|Player|GetPlayerPawn (Fortnite|GetFirstLocalFortPlayerController true)) "(TagName=\"CherrySmoke.HeroRank.A\")")
              (CallFunction|UpdateLight "(R=0.084376,G=1.000000,B=0.144128,A=1.000000)")
              (else
                _))
            (else
              (Variables|Default|SetFirstTimeLoadingThisMatch? true)))
          (CallFunction|UpdateLight "(R=1.000000,G=0.000109,B=0.011584,A=1.000000)"))))))

(event Collision|EventActorBeginOverlap
  (params (OtherActor :type Actor%20Object%20Reference)))

(event EventTick
  (params (DeltaSeconds :type Float%20%28single-precision%29)))

(event Custom|UpdateLight
  (params (NewLightColor :type Linear%20Color%20Structure))
  ; Update Light Event
  (Networking|CanExecuteCosmeticEvents
    (:True
      (if (Variables|Default|GetFirstTimeLoadingThisMatch?)
        (Rendering|Components|Light|SetLightColor (Variables|Default|GetBR_Spotlight_Component) NewLightColor true)
        (Rendering|Material|SetVectorParameterValue (Class|BTechArtGlobalLibrary|GetDynamicMaterialInstance (Variables|StaticMeshActor|GetStaticMeshComponent) 1) "Emissive_Color_Multiplier" NewLightColor :!class /Script/Engine.MaterialInstanceDynamic)
        (Development|PrintString "LIGHT UPDATED")))))

(fn ConstructionScript)
